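# Physics asset for the ScalyButcher two-handed pickaxe axe mesh, simulating two dangling straps attached to the weapon handle.
# Capsule bodies cover the handle and each strap bone, with the strap root free to swing and twist against the handle.
# A tightly limited base joint branches into a six-link and a seven-link chain, each joint given swing and twist limits.
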PhysicsAsset /BRCosmetics/Weapons/FORT_Melee/Pickaxe_ScalyButcher/Meshes/ScalyButcher_Axe_PhysicsAsset
bodies {
 Melee_TwoHanded_Handle [
  {shapeType=Capsule, length=111.364, radius=2.7092, center=<0.9866 -0.7188 20.869>, rotation=<rot -0.1752 -90 0.5065>}
  {
   shapeType=Capsule
   length=1.8831
   radius=1.4949
   shapeName=Melee_TwoHanded_Handle_capsule
   center=<0.8002 -3.0082 -35.6325>
   rotation=<rot 0 0 -90.3145>
  }
 ]
 Rdyn_strap_2_3 [{
   shapeType=Capsule
   length=2.751
   radius=1.3767
   shapeName=Rdyn_strap_2_3_capsule
   center=<2.3938 -0.1192 -0.0131>
   rotation=<rot 0 -89.3947 -89.6254>
  }]
 dyn_strap_1_1 [{
   shapeType=Capsule
   length=2.8747
   radius=0.7783
   shapeName=dyn_strap_1_1_capsule
   center=<1.8625 0.8886 -1.2211>
   rotation=<rot -35.7162 -90 -34.1708>
  }]
 dyn_strap_1_2 [{
   shapeType=Capsule
   length=3.4215
   radius=1.0284
   shapeName=dyn_strap_1_2_capsule
   center=<2.1068 -0.0303 0.2538>
   rotation=<rot -2.5066 -88.8577 -87.6346>
  }]
 dyn_strap_1_3 [{
   shapeType=Capsule
   length=2.4188
   radius=1.5167
   shapeName=dyn_strap_1_3_capsule
   center=<2.1368 -0.1498 0.2747>
   rotation=<rot 0 -89.7838 -87.9228>
  }]
 dyn_strap_1_4 [{
   shapeType=Capsule
   length=6.0788
   radius=1.8991
   shapeName=dyn_strap_1_4_capsule
   center=<4.7355 0.2901 0.3355>
   rotation=<rot 0 -89.2388 -89.8125>
  }]
 dyn_strap_1_5 [{
   shapeType=Capsule
   length=7.3868
   radius=1.8676
   shapeName=dyn_strap_1_5_capsule
   center=<3.7286 0.2773 0.4352>
   rotation=<rot -0.5212 -89.185 -90.001>
  }]
 dyn_strap_1_6 [{
   shapeType=Capsule
   length=1.4112
   radius=2.0211
   shapeName=dyn_strap_1_6_capsule
   center=<2.138 0.3973 0.4028>
   rotation=<rot 0.3155 0.9055 -89.0371>
  }]
 dyn_strap_2_1 [{
   shapeType=Capsule
   length=2.235
   radius=0.5596
   shapeName=dyn_strap_2_1_capsule
   center=<1.5216 -0.0147 -0.0344>
   rotation=<rot 0.145 -92.314 -85.7098>
  }]
 dyn_strap_2_2 [{
   shapeType=Capsule
   length=3.4375
   radius=0.7453
   shapeName=dyn_strap_2_2_capsule
   center=<2.429 0.0368 0.0245>
   rotation=<rot -0.1423 90.8168 -90.0035>
  }]
 dyn_strap_2_4 [{
   shapeType=Capsule
   length=6.8101
   radius=1.951
   shapeName=dyn_strap_2_4_capsule
   center=<5.1789 0.0299 0.0186>
   rotation=<rot 0 -88.9815 -89.7756>
  }]
 dyn_strap_2_5 [{
   shapeType=Capsule
   length=5.1659
   radius=2.2708
   shapeName=dyn_strap_2_5_capsule
   center=<4.7843 0.1648 -0.0476>
   rotation=<rot 0 -89.4488 -89.425>
  }]
 dyn_strap_2_6 [{
   shapeType=Capsule
   length=4.8983
   radius=2.6943
   shapeName=dyn_strap_2_6_capsule
   center=<5.0153 -0.197 0.002>
   rotation=<rot 0 -90.3441 -89.5106>
  }]
 dyn_strap_2_7 [{
   shapeType=Capsule
   length=2.5274
   radius=2.2469
   shapeName=dyn_strap_2_7_capsule
   center=<1.7936 -0.309 -0.0125>
   rotation=<rot 0.0125 -0.6274 -89.7069>
  }]
 dyn_strap_base [{
   shapeType=Capsule
   length=0.6591
   radius=0.8374
   shapeName=dyn_strap_base_capsule
   center=<0.3204 0.0541 -0.4827>
   rotation=<rot 75.0643 -174.333 -98.3142>
  }]
 dyn_strap_root [{shapeType=Capsule, length=0, radius=1.032, shapeName=dyn_strap_root_capsule, rotation=<rot 89.6943 90.0585 -179.9784>}]
}
constraints [
 {bone1Name=dyn_strap_2_6, bone2Name=dyn_strap_2_5, swing1LimitDegrees=25, swing2LimitDegrees=30, twistLimitDegrees=25}
 {
  bone1Name=dyn_strap_root
  bone2Name=Melee_TwoHanded_Handle
  swing1LimitDegrees=45
  swing1Motion=Free
  swing2LimitDegrees=45
  swing2Motion=Free
  twistLimitDegrees=45
  twistMotion=Free
 }
 {bone1Name=dyn_strap_base, bone2Name=dyn_strap_root, swing1LimitDegrees=1, swing2LimitDegrees=10, twistLimitDegrees=2}
 {bone1Name=dyn_strap_1_1, bone2Name=dyn_strap_base, swing1LimitDegrees=35, swing2LimitDegrees=35, twistLimitDegrees=15}
 {bone1Name=dyn_strap_1_2, bone2Name=dyn_strap_1_1, swing1LimitDegrees=25, swing2LimitDegrees=25, twistLimitDegrees=15}
 {bone1Name=dyn_strap_1_3, bone2Name=dyn_strap_1_2, swing1LimitDegrees=25, swing2LimitDegrees=30, twistLimitDegrees=25}
 {bone1Name=dyn_strap_1_4, bone2Name=dyn_strap_1_3, swing1LimitDegrees=25, swing2LimitDegrees=30, twistLimitDegrees=25}
 {bone1Name=dyn_strap_1_5, bone2Name=dyn_strap_1_4, swing1LimitDegrees=25, swing2LimitDegrees=30, twistLimitDegrees=25}
 {bone1Name=dyn_strap_2_1, bone2Name=dyn_strap_base, swing1LimitDegrees=35, swing2LimitDegrees=35, twistLimitDegrees=15}
 {bone1Name=dyn_strap_2_2, bone2Name=dyn_strap_2_1, swing1LimitDegrees=25, swing2LimitDegrees=25, twistLimitDegrees=15}
 {bone1Name=dyn_strap_2_5, bone2Name=dyn_strap_2_4, swing1LimitDegrees=25, swing2LimitDegrees=30, twistLimitDegrees=25}
 {bone1Name=dyn_strap_2_4, bone2Name=Rdyn_strap_2_3, swing1LimitDegrees=25, swing2LimitDegrees=30, twistLimitDegrees=25}
 {bone1Name=Rdyn_strap_2_3, bone2Name=dyn_strap_2_2, swing1LimitDegrees=25, swing2LimitDegrees=35, twistLimitDegrees=25}
 {bone1Name=dyn_strap_2_7, bone2Name=dyn_strap_2_6, swing1LimitDegrees=25, swing2LimitDegrees=30, twistLimitDegrees=25}
 {bone1Name=dyn_strap_1_6, bone2Name=dyn_strap_1_5, swing1LimitDegrees=25, swing2LimitDegrees=30, twistLimitDegrees=25}
]
